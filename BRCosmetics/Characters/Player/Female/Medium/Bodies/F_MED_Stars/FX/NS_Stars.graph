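# Niagara cosmetic effect for the F_MED_Stars player body, with a single Orbit emitter of short-lived rotating mesh particles.
# Particles spawn at a fixed small offset, split into two spawn groups that pick between blue and cyan tints, with randomized scale and orientation.
# Lifetime curves drive dynamic material parameters, spin and alpha fade, followed by a custom hue shift.
# User.bisHeadHidden sets each particle's VisibilityTag, linking renderer visibility to head hiding.
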NiagaraSystem /BRCosmetics/Characters/Player/Female/Medium/Bodies/F_MED_Stars/FX/NS_Stars
dependencies {usedDataInterfaces [&NiagaraDataInterfaceCurve], usedDynamicInputs [], usedModules [&SystemState]}
emitters {Orbit {
  emitterUpdateScript [
   &EmitterState(
    Inactive Response=::NewEnumerator1
    Life Cycle Mode=::NewEnumerator1
    Loop Duration=&UniformRangedFloat(Minimum=3, Maximum=6)
    Recalculate Duration Each Loop=true
    UseLoopDelay=false
   )
   &SpawnRate(Use Spawn Probability=false)
  ]
  particleSpawnScript [
   &FN_SetSpawnGroup_v2(Number of Spawn Groups=2)
   /Niagara/Modules/Spawn/Initialization/InitializeParticle(
    Color=&LerpLinearColors(
     StartColor=<rgba 0.0208 0.7213 2 1>
     EndColor=<rgba 0.1832 2 1.8805 1>
     LerpFactor=&MakeFloatFromInt(Int=$Particles.SpawnGroup)
    )
    Lifetime=&RandomRangeFloat(Minimum=1, Maximum=2)
    Mesh Scale=&VectorFromFloat(Value=&UniformRangedFloat(Minimum=0.2, Maximum=0.25))
    Module.Write Color=true
    Module.Write Lifetime=true
    Module.Write Mass=true
    Module.Write Material Random=true
    Module.Write Position=false
    Module.Write Scale=true
    Module.Write SpriteRotation=false
    Module.Write SpriteSize=false
    Position=&SimulationPosition()
    Sprite Size=<5 5>
   )
   &DynamicMaterialParameters(
    Index 0 Param 1=&UniformRangedFloat(Minimum=0.3, Maximum=0.7)
    Module.Param0WriteEnabled=true
    Module.Param1WriteEnabled=false
    Module.Param2WriteEnabled=false
    Module.Param3WriteEnabled=false
   )
   &InitialMeshOrientation(
    Rotation=&RandomRangeVector(Minimum=<0 -0.05 -0.05>, Maximum=<1 0.05 0.05>)
    Use Orientation Vector=false
    Use Rotation Vector=true
   ) {module=InitialMeshOrientation001}
   &SphereLocation(
    MaskSpawnGroup=false
    Offset=<5 0 0>
    Override Seed=false
    Sphere Radius=0
    Surface Only=false
    UseNonUniformScale=false
    UseOffset=true
    UseSphereOrientation=false
   )
  ]
  particleUpdateScript [
   &ParticleState()
   &SolveForcesAndVelocity(Clamp Velocity=false, Limit Acceleration=false)
   &DynamicMaterialParameters(
    Index 0 Param 2=&FloatFromCurve(FloatCurve=DataInterface() {propertyValues {Curve {keys [
         {arriveTangent=-6.5008, interpMode=RCIM_Cubic, leaveTangent=-6.5008, tangentMode=RCTM_User, time=0, value=2}
         <at 0.5 0.8 RCIM_Cubic>
         <at 1 2 RCIM_Cubic>
        ]}}})
    Index 0 Param 3=&FloatFromCurve(FloatCurve=DataInterface() {propertyValues {Curve {keys [{arriveTangent=1.7971, interpMode=RCIM_Cubic, leaveTangent=1.7971, tangentMode=RCTM_User, time=0, value=0}, <at 1 1 RCIM_Cubic>]}}})
    Index 0 Param 4=&FloatFromCurve(FloatCurve=DataInterface() {propertyValues {Curve {keys [
         <at 0 0 RCIM_Cubic>
         {interpMode=RCIM_Cubic, tangentMode=RCTM_User, time=0.2, value=0.5}
         {interpMode=RCIM_Cubic, tangentMode=RCTM_User, time=0.7, value=0.5}
         {arriveTangent=-4.058, interpMode=RCIM_Cubic, leaveTangent=-4.058, tangentMode=RCTM_User, time=1, value=0}
        ]}}})
    Module.Param0WriteEnabled=false
    Module.Param1WriteEnabled=true
    Module.Param2WriteEnabled=true
    Module.Param3WriteEnabled=true
   ) {module=DynamicMaterialParameters001}
   &UpdateMeshOrientation(
    Rotation Coordinate Space=::NewEnumerator3
    Rotation Rate=&FloatFromCurve(
     FloatCurve=DataInterface() {propertyValues {Curve {keys [<at 0.2 1 RCIM_Cubic>, {arriveTangent=-2.6679, interpMode=RCIM_Cubic, leaveTangent=-2.6679, tangentMode=RCTM_User, time=1, value=0}]}}}
     Scale Curve=-1
    )
    Rotation Vector=<0.85 0 0>
   ) {module=UpdateMeshOrientation001}
   &ScaleColor(
    Scale Alpha=&FloatFromCurve(FloatCurve=DataInterface() {propertyValues {Curve {keys [
         <at 0 0 RCIM_Cubic>
         {interpMode=RCIM_Cubic, tangentMode=RCTM_User, time=0.2, value=1}
         {interpMode=RCIM_Cubic, tangentMode=RCTM_User, time=0.5, value=1}
         <at 1 0 RCIM_Cubic>
        ]}}})
    ScaleA=true
    ScaleRGB=true
   ) {module=ScaleColor001}
   /Game/Effects/Niagara/Modules/Color/FN_ColorHueShift(Color:LinearColor=<rgba 0.3961 0.0902 1 1>)
   /SetVariables(Particles.VisibilityTag=&MakeCustomIntFromBool(Bool=$User.bisHeadHidden, True INT=1, False INT=0)) {isSetParameters=true}
  ]
  rendererClasses [&NiagaraMeshRendererProperties]
  renderers [{rendererClass=&NiagaraMeshRendererProperties, rendererIndex=0}]
 }}
summary {
 systemName=NS_Stars
 emitters [{emitterName=Orbit, rendererClasses [&NiagaraMeshRendererProperties]}]
 userVariables [
  User.bisHeadHidden {type=NiagaraBool}
  User.NewNiagaraDataInterfaceSkeletalMesh {
   type=NiagaraDataInterfaceSkeletalMesh
   defaultValue {
    struct=NiagaraExt_VariableValue_DataInterface
    value {dataInterface=&self:NiagaraDataInterfaceSkeletalMesh_0}
   }
  }
 ]
}
